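; BP_BeetCalf_Console is a console Blueprint whose screen switches between available, denied and complete states through material parameters.
; Each SetScreen function stores a state index and calls UpdateScreenStateVisuals with static enabled.
; That event shows TV static for StaticScreenDuration, then clears it and applies the clamped index.
; Both visuals are scalar parameters on a dynamic material instance of the static mesh, fetched via BTechArtGlobalLibrary.
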(event EventBeginPlay)

(event Collision|EventActorBeginOverlap
  (params (OtherActor :type Actor%20Object%20Reference)))

(event EventTick
  (params (DeltaSeconds :type Float%20%28single-precision%29)))

(event Custom|UpdateScreenStateVisuals
  (params (Index :type Integer) (ShowTVStatic :type Boolean))
  ;[1 1 1] Show Static -> Wait -> Remove Static -> Show new Screen Index
  ;
  ; Or Show a different screen without static in between
  (if ShowTVStatic
    (Screen|ToggleMaterialTVStatic true)
    (Utilities|FlowControl|Delay (Variables|Default|GetStaticScreenDuration))
    (else
      _))
  (Screen|ToggleMaterialTVStatic false)
  (Screen|UpdateMaterialScreenStateIndex Index))

(fn SetScreenAvailable
  (bind _output_get (Variables|Default|SetScreenStateIndex 0))
  (CallFunction|UpdateScreenStateVisuals _output_get true))

(fn SetScreenComplete
  (bind _output_get (Variables|Default|SetScreenStateIndex 2))
  (CallFunction|UpdateScreenStateVisuals _output_get true))

(fn SetScreenDenied
  (bind _output_get (Variables|Default|SetScreenStateIndex 1))
  (CallFunction|UpdateScreenStateVisuals _output_get true))

(fn ToggleMaterialTVStatic
  (params (bShowStatic :type Boolean))
  (Rendering|Material|SetScalarParameterValue (Class|BTechArtGlobalLibrary|GetDynamicMaterialInstance (Variables|Debug|GetStaticMeshComponent)) "BeetCalf.ShowTVStaticNoise" (Math|Conversions|ToFloat(Integer) (Math|Integer|SelectInt 1 0 bShowStatic)) :!class /Script/Engine.MaterialInstanceDynamic))

(fn UpdateMaterialScreenStateIndex
  (params (Index :type Integer))
  (Rendering|Material|SetScalarParameterValue (Class|BTechArtGlobalLibrary|GetDynamicMaterialInstance (Variables|Debug|GetStaticMeshComponent)) "BeetCalf.StateIndex" (Math|Conversions|ToFloat(Integer) (Math|Integer|Clamp(Integer) Index 0 2)) :!class /Script/Engine.MaterialInstanceDynamic))

(fn ConstructionScript
  (|CallParentFunction))
